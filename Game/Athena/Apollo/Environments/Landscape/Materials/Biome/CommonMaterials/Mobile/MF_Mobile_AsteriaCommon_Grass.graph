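# Material function that builds Asteria biome grass attributes for mobile landscape, static meshes and grass meshes, and also outputs height.
# It draws color, roughness, specular and normal data from MF_Asteria_Common_Grass, while a soft-slope mask from MF_Mobile_Common_Cliff darkens grass on slopes.
# Shading path switches give Forward and Mobile simpler variants, mostly using vertex normals.
# The debug biome-color branch and static-switch color override feed no output.
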MaterialFunction /Game/Athena/Apollo/Environments/Landscape/Materials/Biome/CommonMaterials/Mobile/MF_Mobile_AsteriaCommon_Grass
graph {nodes [
  Constant() {r=0.9}
  Constant() {r=0.3}
  Constant() {r=0.3}
  n1 = NamedRerouteDeclaration(None=Constant() {r=0.65}) {name=SoftSlopeScalar, nodeColor=<rgba 1 0 1 1>}
  n2 = NamedRerouteUsage()
  n3 = MaterialFunctionCall!(UseParameters) {materialFunction=&/MF_Mobile_Common_Cliff}
  n4 = SetMaterialAttributes!(
   MaterialAttributes
   BaseColor=LinearInterpolate(A=Multiply(A=@n1, B=@n2), B=@n2, Alpha=NamedRerouteDeclaration(None=@n3.SoftSlope) {name=SoftSlopeMask, nodeColor=<rgba 1 0 1 1>})
   Specular=NamedRerouteUsage()
   Roughness=NamedRerouteUsage()
   Normal=VertexNormalWS()
  ) {attributeSetTypes [
    #1
    #2
    #3
    #4
   ]}
  FunctionOutput(None=@n4) {desc=Mobile Landscape Only, outputName=MobileLandscapeMaterial, sortPriority=-1}
  n5 = SetMaterialAttributes(BaseColor=NamedRerouteUsage(), Specular=NamedRerouteUsage(), Roughness=NamedRerouteUsage(), Normal=NamedRerouteUsage()) {attributeSetTypes [
    #1
    #2
    #3
    #4
   ]}
  n6 = SetMaterialAttributes(BaseColor=NamedRerouteUsage(), Specular=NamedRerouteUsage(), Roughness=NamedRerouteUsage(), Normal=VertexNormalWS()) {attributeSetTypes [
    #1
    #2
    #3
    #4
   ]}
  FunctionOutput(None=ShadingPathSwitch(Default=@n5, Deferred=@n5, Forward=@n6, Mobile=@n6) {desc="This output is used for mobile grass,\r\nthe grass spawned on static meshes\r\n&& the grass terrain material on static meshes"}) {outputName=ForStaticMesh, sortPriority=80}
  n7 = NamedRerouteUsage()
  n8 = LinearInterpolate(A=Multiply(A=Constant() {desc="Mobile grass mesh\r\nSoftSlopeScalar", r=0.8}, B=@n7), B=@n7, Alpha=NamedRerouteUsage()) {desc=Makes grass darker on slopes}
  n9 = ShadingPathSwitch(Default=NamedRerouteUsage(), Forward=@n8, Mobile=@n8)
  n10 = SetMaterialAttributes(
   BaseColor=@n9
   Specular=ScalarParameter() {defaultValue=0.15, group=MF_MCG, parameterName=ForGrassMesh-Spec}
   Roughness=ScalarParameter() {defaultValue=0.9, group=MF_MCG, parameterName=ForGrassMesh-Rough}
   Normal=VertexNormalWS()
  ) {attributeSetTypes [
    #1
    #2
    #3
    #4
   ]}
  n11 = SetMaterialAttributes(BaseColor=@n9, Specular=Constant() {r=0.3}, Roughness=Constant() {r=0.9}, Normal=VertexNormalWS()) {attributeSetTypes [
    #1
    #2
    #3
    #4
   ]}
  FunctionOutput(None=ShadingPathSwitch(Default=@n10, Forward=@n11, Mobile=@n11)) {outputName=ForGrassMeshes, sortPriority=80}
  n12 = FunctionInput(Preview=Constant()) {
   bUsePreviewValueAsDefault=true
   desc=Input the mask for the grass edge color.
   description=Input the mask for the grass edge color.
   inputName=CombinedLandscapeLayersMask
   inputType=FunctionInput_Scalar
  }
  n13 = MaterialFunctionCall!(BiomeMask-RGBA, Grass Localised Tint Mask, CombinedLandscapeLayersMask=@n12, BiomeColors-UseDefault) {materialFunction=/Game/Athena/Asteria/Environments/Landscape/Materials/MaterialFunctions/MF_Asteria_Common_Grass}
  NamedRerouteDeclaration(None=@n13.Mobile-Landscape-Color) {name=Mobile-Landscape-Color, nodeColor=<rgba 1 0 1 0>}
  NamedRerouteDeclaration(None=@n13.Default-Landscape-Color) {name=Default-Landscape-Color, nodeColor=<rgba 1 0 1 1>}
  n14 = NamedRerouteUsage()
  n15 = LinearInterpolate(A=Multiply(A=NamedRerouteUsage(), B=@n14), B=@n14, Alpha=NamedRerouteUsage()) {desc=Makes grass darker on slopes}
  NamedRerouteDeclaration(None=ShadingPathSwitch(Default=NamedRerouteUsage(), Forward=@n15, Mobile=@n15)) {name=StaticMeshBaseColor, nodeColor=<rgba 1 0 1 1>}
  n16 = VectorParameter() {group=MF_MCG, parameterName=MF_MCG_Override-Basecolor, defaultValue=<rgba 1 1 1 1>}
  NamedRerouteDeclaration(None=Saturate(None=Divide(A=@n13.GrassMeshColor, B=@n16.RGB))) {name=GrassMeshColor, nodeColor=<rgba 1 0 1 1>}
  NamedRerouteDeclaration(None=FunctionOutput(None=@n13.Mobile-Height) {outputName=Height}) {name=Height, nodeColor=<rgba 1 0 1 1>}
  NamedRerouteDeclaration(None=Saturate(None=Divide(A=@n13.BasicGrassMeshColor, B=@n16.RGB))) {name=BasicGrassMeshColor, nodeColor=<rgba 1 0 1 1>}
  NamedRerouteDeclaration(None=@n13.Spec-Basic) {name=BasicSpec, nodeColor=<rgba 1 0 1 1>}
  NamedRerouteDeclaration(None=@n13.Rough-Basic) {name=BasicRough, nodeColor=<rgba 1 0 1 1>}
  NamedRerouteDeclaration(None=@n13.Spec-CameraDistance) {name=Spec-NoRVT, nodeColor=<rgba 1 0 1 1>}
  NamedRerouteDeclaration(None=@n13.WorldNormals) {name=WorldNormals, nodeColor=<rgba 1 0 1 1>}
  n17 = ScalarParameter() {defaultValue=0.9, parameterName=debuggg-363f23986}
  n18 = LinearInterpolate(A=NamedRerouteUsage(), B=Multiply(A=@n13.GrassMeshColor), Alpha=ScalarParameter() {defaultValue=1, parameterName=debuggg-363562f})
  n19 = MaterialFunctionCall!(Grass=@n18, Gravel, Mud, Sand, BiomeMask-RGBA, UseBiomeColorMask, UseLandscapeCoords) {materialFunction=/Game/Athena/Asteria/Environments/Landscape/Materials/MaterialFunctions/MF_AsteriaBiomeColors}
  n20 = VertexNormalWS()
  n21 = Power(Base=Abs(None=ComponentMask(None=@n20) {b=true}), Exp=Constant() {r=3})
  n22 = Multiply(
   A=Multiply(A=OneMinus(None=@n21), B=ScalarParameter() {defaultValue=0.5, parameterName=debuggg-t486yethku4hhf54})
   B=ScalarParameter() {defaultValue=1, parameterName=debuggg-t486yethku4grgw}
  )
  n23 = Add(A=Multiply(A=Constant() {r=0.01}, B=@n17), B=Multiply(A=@n17, B=LinearInterpolate(A=@n18, B=@n19.GrassEdge, Alpha=@n22)))
  Reroute(None=LinearInterpolate(A=Saturate(None=@n23), B=@n9, Alpha=ScalarParameter() {parameterName=debuggg-kjhkjkjw234z}))
  n24 = VectorParameter() {group=MF_MCG, parameterName=MF_MCG_Override-SoftSlopeColor, defaultValue=<rgba 0.1 0.1 0.1 1>}
  n25 = VectorParameter() {group=MF_MCG, parameterName=MF_MCG_Override-Basecolor, defaultValue=<rgba 1 1 1 1>}
  n26 = NamedRerouteUsage()
  StaticSwitchParameter(True=Saturate(None=LinearInterpolate(A=@n24.RGB, B=@n25.RGB, Alpha=@n26)), False=@n9) {desc="True= Used on grass meshes that want to use a custom base color", group=MF_MCG, parameterName=MF_MCG_Override-StaticMesh-Basecolor}
 ]}
