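# Mobile landscape material function that layers the common Apollo biome surfaces into one material attributes output.
# It height-blends grass, gravel, mud (recolored gravel), sand, asphalt and cliff in sequence, with grass receiving a combined mask of painted layers.
# Static switches allow moving sand before gravel and adding procedural snow on top, keeping vertex attributes from the base.
# A height-darkening and sun-shadow tint branch is left unconnected to the output.
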MaterialFunction /Game/Athena/Apollo/Environments/Landscape/Materials/Biome/CommonMaterials/Mobile/MF_Mobile_CommonLandscapeLayers
graph {nodes [
  n1 = LandscapeLayerSample() {parameterName=Mud}
  n2 = LandscapeLayerSample() {parameterName=Gravel}
  n3 = Max(A=Max(A=Max(A=@n1, B=@n2), B=LandscapeLayerSample() {parameterName=Sand}), B=LandscapeLayerSample() {parameterName=Asphalt})
  n4 = Constant() {r=3}
  n5 = MaterialFunctionCall!(CombinedLandscapeLayersMask=Saturate(None=Multiply(A=@n3, B=@n4)), BypassUseRVT, BypassBiomeColors) {materialFunction=&/MF_Mobile_Common_Grass}
  n6 = MaterialFunctionCall!(CustomSeaLevel, BypassBiomeColors, UseOceanLevel=NamedRerouteUsage()) {materialFunction=&/MF_Mobile_Common_Sand}
  n7 = MaterialFunctionCall!(
   Material 1=@n5.MobileLandscapeMaterial
   Height Texture 1=@n5.MobileLandscapeMaterial
   Material 2=@n6.Material
   Height Texture 2=@n6.Material
   Transistion Phase=LandscapeLayerSample() {parameterName=Sand}
   UseContrast
   Contrast
   AddLayerMask
  ) {materialFunction=/Game/Athena/Apollo/Environments/Landscape/Materials/Biome/CommonMaterials/MF_BlendLandscapeLayers}
  n8 = NamedRerouteUsage()
  n9 = LandscapeLayerSample() {parameterName=Gravel}
  n10 = MaterialFunctionCall!(OverrideEdgeColor, EdgeMask=@n9, BypassBiomeColors) {materialFunction=&/MF_Mobile_Common_Gravel}
  n11 = MaterialFunctionCall(
   Material 1=StaticSwitch(True=@n7.BlendedMaterial, False=@n5.MobileLandscapeMaterial, Value=@n8)
   Height Texture 1=StaticSwitch(True=@n7.Resulting Height Map, False=@n5.Height, Value=@n8)
   Material 2=@n10.Material
   Height Texture 2=@n10.Material
   Transistion Phase=@n9
  ) {materialFunction=&MF_BlendLandscapeLayers}
  n12 = StaticBool() {value=true}
  n13 = MaterialFunctionCall(
   Material 1=@n11.BlendedMaterial
   Height Texture 1=@n11.BlendedMaterial
   Material 2=SetMaterialAttributes!(
    MaterialAttributes=@n10.Material
    BaseColor=MaterialFunctionCall!(GravelMaterialInput, Heightmap, UseUnmodifiedGravelDiffuse=@n12, UnmodifiedGravelDiffuse=@n10.RGB, BypassBiomeColors) {materialFunction=/Game/Athena/Apollo/Environments/Terrain_Meshes/MaterialFunctions/MF_DirtMudModifier}.ColorizedGravelDiffuse
   ) {attributeSetTypes [#1]}
   Height Texture 2=@n10.Height
   Transistion Phase=LandscapeLayerSample() {parameterName=Mud}
  ) {materialFunction=&MF_BlendLandscapeLayers}
  n14 = MaterialFunctionCall(UseOceanLevel=NamedRerouteUsage()) {materialFunction=&MF_Mobile_Common_Sand}
  n15 = MaterialFunctionCall(
   Material 1=@n13.BlendedMaterial
   Height Texture 1=@n13.BlendedMaterial
   Material 2=@n14.Material
   Height Texture 2=@n14.Material
   Transistion Phase=LandscapeLayerSample() {parameterName=Sand}
  ) {materialFunction=&MF_BlendLandscapeLayers}
  n16 = NamedRerouteUsage()
  n17 = MaterialFunctionCall!(TilingSize) {materialFunction=/Game/Athena/Apollo/Environments/Landscape/Materials/MaterialFunctions/MF_ApolloAsphalt}
  n18 = MaterialFunctionCall(
   Material 1=StaticSwitch(True=@n13.BlendedMaterial, False=@n15.BlendedMaterial, Value=@n16)
   Height Texture 1=StaticSwitch(True=@n13.Resulting Height Map, False=@n15.Resulting Height Map, Value=@n16)
   Material 2=@n17.MobileLandscapeMaterial
   Height Texture 2=@n17.MobileLandscapeMaterial
   Transistion Phase=LandscapeLayerSample() {parameterName=Asphalt}
  ) {materialFunction=&MF_BlendLandscapeLayers}
  n19 = MaterialFunctionCall!(UseParameters) {materialFunction=&/MF_Mobile_Common_Cliff}
  n20 = MaterialFunctionCall(
   Material 1=@n18.BlendedMaterial
   Height Texture 1=@n18.BlendedMaterial
   Material 2=@n19.LandscapeMaterial
   Height Texture 2=@n19.LandscapeMaterial
   Transistion Phase=@n19.LandscapeMaterial
  ) {materialFunction=&MF_BlendLandscapeLayers}
  n21 = MaterialFunctionCall!(
   Heightmap=@n20.Resulting Height Map
   WorldnormalIn=GetMaterialAttributes(None=@n20.BlendedMaterial) {attributeGetTypes [#2]}.Normal
   UseManualSnowMask
   SnowMask
  ) {materialFunction=/Game/Athena/Apollo/Environments/Landscape/Materials/Biome/UniqueBiomeMaterialFunctions/MF_ProceduralSnow}
  n22 = StaticBool()
  FunctionOutput(None=StaticSwitch(
    True=BlendMaterialAttributes(A=@n20.BlendedMaterial, B=@n21.MobileMaterial, Alpha=@n21.MobileMaterial) {desc="Vertex blend type altered!", vertexAttributeBlendType=UseA}
    False=@n20.BlendedMaterial
    Value=FunctionInput(Preview=@n22) {bUsePreviewValueAsDefault=true, inputName=UseSnow, inputType=FunctionInput_StaticBool, sortPriority=1}
   )) {outputName=Material, sortPriority=-1}
  n23 = StaticBool()
  NamedRerouteDeclaration(None=StaticSwitch(
    True=StaticBool()
    False=StaticBoolParameter() {group=Mobile, parameterName=Mobile-PutSandBeforeGravel}
    Value=FunctionInput(Preview=@n23) {bUsePreviewValueAsDefault=true, inputName=PutSandBeforeGravelOverride, inputType=FunctionInput_StaticBool}
   )) {name=PutSandBeforeGravel, nodeColor=<rgba 0.5 0 1 1>}
  NamedRerouteDeclaration(None=FunctionInput(Preview=StaticBool() {value=true}) {bUsePreviewValueAsDefault=true, inputName=UseOceanLevel, inputType=FunctionInput_StaticBool, sortPriority=100}) {name=UseOceanLevel, nodeColor=<rgba 0.5 0 1 1>}
  ScalarParameter() {defaultValue=0.5, parameterName=MFMCL-ShadowAdd}
  n24 = GetMaterialAttributes(None=@n20.BlendedMaterial) {attributeGetTypes [#1]}
  n25 = ComponentMask(None=WorldPosition() {worldPositionShaderOffset=WPT_ExcludeAllShaderOffsets}.XYZ) {b=true}
  n26 = Constant() {r=40}
  n27 = Saturate(None=Divide(A=Add(A=@n25, B=@n26), B=Constant() {r=75}))
  n28 = Constant() {r=0.5}
  n29 = Multiply(A=Saturate(None=Add(A=@n27, B=@n28)), B=@n24.BaseColor)
  n30 = OneMinus(None=SkyAtmosphereLightDirection())
  n31 = ScalarParameter() {defaultValue=4, parameterName=MFMCL-ShadowPow}
  n32 = Saturate(None=DotProduct(A=PixelNormalWS(), B=Power(Base=@n30, Exp=@n31)))
  n33 = Multiply(A=VectorParameter() {parameterName=MFMCL-ShadowColor, defaultValue=<rgba 30 30 30 1>}.RGB, B=@n32)
  SetMaterialAttributes(MaterialAttributes=@n24.MaterialAttributes, BaseColor=LinearInterpolate(A=@n29, B=Multiply(A=@n33, B=@n29), Alpha=@n32)) {attributeSetTypes [#1]}
  SetMaterialAttributes(MaterialAttributes=@n24.MaterialAttributes, BaseColor=@n29) {attributeSetTypes [#1]}
 ]}
